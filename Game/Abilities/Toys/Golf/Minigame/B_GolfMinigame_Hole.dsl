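; Golf minigame hole actor that detects sunk balls, reports shots to the message feed and grants quest progress.
; On the server, green landings and cup overlaps read stroke count and XY distance into a replicated message struct, with an incrementing serial number.
; Its OnRep formats and prints client text, with distance setting squad or team relevance.
; A missing greens POI triggers a slow tag scan of POI volumes at BeginPlay.
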(event EventBeginPlay
  (Networking|SwitchHasAuthority
    (:Authority
      (Utilities|IsValid (Variables|Config|GetGreensPOI)
        (:Is%20Not%20Valid
          (CallFunction|SLOWScanforMatchingGreens))))))

(event Custom|LandedOnGreen
  (params (GolfBall :type B%20Toy%20Golf%20Ball%20Base%20Object%20Reference))
  (bind _returnvalue (Actor|GetOwner GolfBall))
  ;[1 1 1] Land Ball from Tee to Green
  (Utilities|FlowControl|Sequence
    (:then_0
      (bind (_numstrokes _iscorrecthole) (Class|BToyGolfBallBase|GetStrokeCountforHole GolfBall (Variables|Config|GetHoleNumber)))
      (bind _asfort_player_state_athena (Utilities|Casting|CastToFortPlayerStateAthena (Actor|GetOwner GolfBall))
        (:then
          _)
        (:CastFailed
          _))
      (CallFunction|CreateMessage _asfort_player_state_athena (CallFunction|GetBalltoPinDistanceXY GolfBall) _numstrokes _iscorrecthole))
    (:then_1
      (Utilities|IsValid _returnvalue
        (:Is%20Valid
          (CallFunction|GrantQuestReward (Actor|GetOwner _returnvalue)))))
    (:then_2
      (Development|PrintString "ShotFromTee?" true true "(R=0.394330,G=1.000000,B=0.000000,A=1.000000)" 1e+01))))

(event OnComponentBeginOverlap(HoleCollision)
  (params (OverlappedComponent :type Primitive%20Component%20Object%20Reference) (OtherActor :type Actor%20Object%20Reference) (OtherComp :type Primitive%20Component%20Object%20Reference) (OtherBodyIndex :type Integer) (bFromSweep :type Boolean) (SweepResult :type Hit%20Result%20Structure%20%28by%20ref%29))
  ;[1 1 1] Hole in One
  (Networking|SwitchHasAuthority
    (:Authority
      (bind _asb_toy_golf_ball_base (Utilities|Casting|CastToB_Toy_GolfBall_Base OtherActor))
      (bind _asfort_player_state_athena (Utilities|Casting|CastToFortPlayerStateAthena (Actor|GetOwner _asb_toy_golf_ball_base))
        (:then
          _)
        (:CastFailed
          _))
      (bind (_numstrokes _iscorrecthole) (Class|BToyGolfBallBase|GetStrokeCountforHole _asb_toy_golf_ball_base (Variables|Config|GetHoleNumber)))
      (CallFunction|CreateMessage _asfort_player_state_athena (CallFunction|GetBallFlightDistanceXY _asb_toy_golf_ball_base) _numstrokes _iscorrecthole true)
      (Class|BToyGolfBallBase|PlayGolfScoringCueandDestruct _asb_toy_golf_ball_base (Variables|Config|GetHoleNumber))
      (if (== (Variables|Config|GetHoleNumber) (Variables|/Game/Abilities/Toys/Golf/Shared/B_Toy_GolfBall_Base.B_Toy_GolfBall_Base_C|GetStartedAtTeeBoxNumber _asb_toy_golf_ball_base))
        (CallFunction|GrantQuestReward (Actor|GetOwner (Actor|GetOwner _asb_toy_golf_ball_base)))))))

(event Custom|Quest%20Reward%20FX
  (bind _playerwhocompletedquest (Variables|Default|GetPlayerWhoCompletedQuest))
  ;[1 1 1] Quest Success FX
  (Networking|SwitchHasAuthority
    (:Remote
      (Utilities|IsValid _playerwhocompletedquest
        (:Is%20Valid
          (Audio|PlaySoundatLocation (Variables|Default|GetQuestSuccessSound) (Transformation|GetActorLocation _playerwhocompletedquest))
          (Effects|SpawnFXSystemAttached 0 (Variables|/Script/Engine.Character|GetMesh (Game|Player|GetPlayerPawn _playerwhocompletedquest))))))))

(fn CreateMessage
  (params (PlayerWhoShot :type Fort%20Player%20State%20Object%20Reference) (TotalDistance :type Float%20%28double-precision%29) (NumStrokes :type Integer) (RightHole :type Boolean) (HoleInOne :type Boolean))
  (if (Variables|/Script/FortniteGame.FortRuntimeOptions|GetbToyMessagingEnabled (Options|GetRuntimeOptions))
    (Variables|Default|SetDisplayMessage (Utilities|Struct|MakeStructGolfMinigameMessage PlayerWhoShot TotalDistance NumStrokes RightHole HoleInOne (Variables|Default|GetMessageSN)))
    (Variables|Default|SetMessageSN (Math|Byte|Byte+Byte (Variables|Default|GetMessageSN) 1))))

(fn GetBallFlightDistanceXY
  (params (Target :type Physics%20Ball%20Master%20Object%20Reference))
  (return (/ (Math|Vector|VectorLengthXY (- (Transformation|GetWorldLocation (Variables|Default|GetHoleCollision)) (Variables|/Game/Building/ActorBlueprints/Prop/PhysicsBall_Master.PhysicsBall_Master_C|GetInitialSpawnPosition Target))) 1e+02)))

(fn GetBallToPinDistanceXY
  (params (Target :type Physics%20Ball%20Master%20Object%20Reference))
  (return (/ (Math|Vector|VectorLengthXY (- (Transformation|GetWorldLocation (Variables|Default|GetHoleCollision)) (Transformation|GetActorLocation Target))) 1e+02)))

(fn GetDistanceText
  (bind _displaymessage (Variables|Default|GetDisplayMessage))
  (bind (_playerwhoshot_30_ef5d780b4020e911ab429d9ba6b977ba _totaldistance_13_a600d2a048cb01620b914bb4107a6cd5) (Utilities|Struct|BreakStructGolfMinigameMessage _displaymessage))
  (if (> _totaldistance_13_a600d2a048cb01620b914bb4107a6cd5 5.0)
    (return (Utilities|Text|ToText(Integer) (Math|Float|Round _totaldistance_13_a600d2a048cb01620b914bb4107a6cd5)))
    (else
      (return (Utilities|Text|ToText(Float) _totaldistance_13_a600d2a048cb01620b914bb4107a6cd5 "HalfToEven" false true 1 324 1 1)))))

(fn GrantQuestReward
  (params (InActor :type Actor%20Object%20Reference))
  (bind _cancompletequest (Class|AthenaQuestFunctionLibrary|AthenaCheckCreativeMode (Game|GetGameState)))
  (if _cancompletequest
    (bind _asfort_player_controller (Utilities|Casting|CastToFortPlayerController InActor))
    (bind (_questisvalid _questcompleted) (Default|AthenaCheckQuestandObjectiveCompletion _asfort_player_controller (Variables|Quest|GetQuestItem) (Variables|Quest|GetObjectiveBackendName)))
    (if (and _questisvalid (not _questcompleted))
      (Variables|Default|SetPlayerWhoCompletedQuest _asfort_player_controller)
      (Quest|SendCustomStatEventDirect (Quests|GetQuestManager _asfort_player_controller "Invalid") (Variables|Quest|GetObjectiveBackendName) (Quests|GetQuestwithDefinition (Quests|GetQuestManager _asfort_player_controller "Invalid") (Variables|Quest|GetQuestItem)) 0)
      (CallFunction|QuestRewardFX))))

(fn OnRep_DisplayMessage
  (bind _returnvalue (BlueprintContext|GetFortToyContext))
  (bind (_displaymessage_playerwhoshot_30_ef5d780b4020e911ab429d9ba6b977ba _displaymessage_totaldistance_13_a600d2a048cb01620b914bb4107a6cd5 _displaymessage_numstrokes_21_9facd4014a6f7fdaaa915396bd79b184) (Utilities|Struct|BreakStructGolfMinigameMessage (Variables|GetDisplayMessage_PlayerWhoShot_30_EF5D780B4020E911AB429D9BA6B977BA)))
  (bind _displaymessage_playerwhoshot_30_ef5d780b4020e911ab429d9ba6b977ba_2 (Utilities|Struct|BreakStructGolfMinigameMessage (Variables|GetDisplayMessage_PlayerWhoShot_30_EF5D780B4020E911AB429D9BA6B977BA)))
  (Networking|SwitchHasAuthority
    (:Remote
      (Utilities|IsValid _displaymessage_playerwhoshot_30_ef5d780b4020e911ab429d9ba6b977ba
        (:Is%20Valid
          (bind _returnvalue_2 (Class|FortToyContext|FormatPlayerNameforDisplay (BlueprintContext|GetFortToyContext) _displaymessage_playerwhoshot_30_ef5d780b4020e911ab429d9ba6b977ba "ToyOwner"))
          (bind _outformatstring (CallFunction|PickFormatString))
          (bind _distanceastext (CallFunction|GetDistanceText))
          (Utilities|IsValid _returnvalue
            (:Is%20Valid
              (Class|FortToyContext|ConditionalPrinttoMessageFeed _returnvalue _displaymessage_playerwhoshot_30_ef5d780b4020e911ab429d9ba6b977ba_2 (Transformation|GetActorLocation) (select (> _displaymessage_totaldistance_13_a600d2a048cb01620b914bb4107a6cd5 (Variables|Default|GetMinDistanceForTeamRelevance)) "SquadAndTeamMembers" "SquadMembersOnly") (Utilities|Text|FormatText _outformatstring _returnvalue_2 _distanceastext _displaymessage_numstrokes_21_9facd4014a6f7fdaaa915396bd79b184 (Variables|Config|GetHoleNumber))))))))))

(fn PickFormatString
  (bind (_displaymessage_playerwhoshot_30_ef5d780b4020e911ab429d9ba6b977ba _displaymessage_totaldistance_13_a600d2a048cb01620b914bb4107a6cd5 _displaymessage_numstrokes_21_9facd4014a6f7fdaaa915396bd79b184 _displaymessage_righthole_25_54415216430c45042af683a06f9801e1 _displaymessage_holeinone_32_0b07a20e436a8d8470698da81073e18f) (Utilities|Struct|BreakStructGolfMinigameMessage (Variables|GetDisplayMessage_PlayerWhoShot_30_EF5D780B4020E911AB429D9BA6B977BA)))
  (if _displaymessage_holeinone_32_0b07a20e436a8d8470698da81073e18f
    (if _displaymessage_righthole_25_54415216430c45042af683a06f9801e1
      (return (select (== _displaymessage_numstrokes_21_9facd4014a6f7fdaaa915396bd79b184 1) (Variables|Default|GetFinishedRightHole_InOne) (Variables|Default|GetFinishedRightHole_MultipleStrokes)))
      (else
        (return (Variables|Default|GetFinishedWrongHole))))
    (else
      (return (Variables|Default|GetOnGreen_InOne)))))

(fn SLOW_ScanForMatchingGreens
  (Development|PrintString (Utilities|String|Append "WARNING: Golf course hole " (Utilities|String|ToString(Integer) (Variables|Config|GetHoleNumber)) " has no green associated with it, and is scanning to find one.  This is slow!") true true "(R=1.000000,G=0.000000,B=0.070032,A=1.000000)" 4.0)
  (bind _outactors (Actor|GetAllActorsOfClass "/Script/FortniteGame.FortPoiVolume"))
  (for _array_element _outactors
    (if (Class|FortPoiVolume|ContainsLocationTag _array_element (Variables|Config|GetGreensPOITag))
      (Variables|Config|SetGreensPOI _array_element)
      (return))))

(fn ConstructionScript
  (bind _holeeditortext (Variables|Default|GetHoleEditorText))
  (|CallParentFunction)
  (Utilities|IsValid _holeeditortext
    (:Is%20Valid
      (Rendering|Components|TextRender|SetText _holeeditortext (Utilities|Text|FormatText "INVTEXT(\"Hole {HoleNumber}\")" (Variables|Config|GetHoleNumber))))))
