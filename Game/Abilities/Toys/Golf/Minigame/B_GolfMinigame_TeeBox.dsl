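; Tee box Blueprint for the golf minigame, defining a configurable box area tied to a hole number.
; The construction script sizes the tee box collision to half the configured tee box size, since box extents are half-dimensions.
; When the editor text component is valid, it also labels it with the hole number and "Tee Box" on two lines.
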(fn ConstructionScript
  (bind _teeboxeditortext (Variables|Default|GetTeeBoxEditorText))
  (Components|Box|SetBoxExtent (Variables|Default|GetTeeBoxCollision) (/ (Variables|Config|GetTeeBoxSize) 2.0) false)
  (Utilities|IsValid _teeboxeditortext
    (:Is%20Valid
      (Rendering|Components|TextRender|SetText _teeboxeditortext (Utilities|Text|FormatText "INVTEXT(\"Hole {HoleNumber}\\r\\nTee Box\")" (Variables|Config|GetHoleNumber))))))
